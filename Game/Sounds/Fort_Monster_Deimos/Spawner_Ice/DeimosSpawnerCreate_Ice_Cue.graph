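# Sound cue for the creation of an ice Deimos monster spawner, judging by its name and waves.
# A Random node picks between two equally weighted spawn waves and feeds an Attenuation node that uses the shared large ambient attenuation of the Deimos spawner.
# It plays boosted in the mission in-game sound class and overrides concurrency to cap simultaneous instances, stopping the farthest before preventing new ones.
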SoundCue /Game/Sounds/Fort_Monster_Deimos/Spawner_Ice/DeimosSpawnerCreate_Ice_Cue
root=Attenuation_0
nodes [
 {
  id=Attenuation_0
  children [Random_1]
  properties {attenuationSettings=/Game/Sounds/Fort_Monster_Deimos/Spawner/Fort_Deimos_SpawnerAmbientLarge_Att}
 }
 {id=Random_1, children [WavePlayer_0,WavePlayer_1], properties {weights [1,1]}}
 {id=WavePlayer_0, properties {soundWaveAssetPtr=&/Deimos_SpawnerSpawn_Ice_01}}
 {id=WavePlayer_1, properties {soundWaveAssetPtr=&/Deimos_SpawnerSpawn_Ice_02}}
]
properties {
 attenuationSettings=&Fort_Deimos_SpawnerAmbientLarge_Att
 bOverrideConcurrency=true
 maxDistance=4000
 soundClassObject=/Game/Sounds/FortSoundClassesAndModes/Fort_Mission_IngameSounds
 volumeMultiplier=2
 attenuationOverrides {falloffDistance=1000000}
 concurrencyOverrides {maxCount=4, resolutionRule=StopFarthestThenPreventNew}
}
