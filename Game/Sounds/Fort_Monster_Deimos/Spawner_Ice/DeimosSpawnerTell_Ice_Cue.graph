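# Sound cue for the charge-up tell of the ice variant of the Deimos monster spawner.
# A Random node picks one of two equally weighted ice charge waves, is excluded from branch culling, and feeds a Modulator at default settings.
# It borrows the large spawner ambient attenuation, routes to the mission in-game sound class, and overrides concurrency to reject new plays beyond a small limit.
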SoundCue /Game/Sounds/Fort_Monster_Deimos/Spawner_Ice/DeimosSpawnerTell_Ice_Cue
root=Modulator_2
nodes [
 {id=Modulator_2, children [Random_1]}
 {id=Random_1, children [WavePlayer_5,WavePlayer_6], properties {bShouldExcludeFromBranchCulling=true, weights [1,1]}}
 {id=WavePlayer_5, properties {soundWaveAssetPtr=&/Deimos_SpawnerCharge_Ice_01}}
 {id=WavePlayer_6, properties {soundWaveAssetPtr=&/Deimos_SpawnerCharge_Ice_02}}
]
properties {
 attenuationSettings=/Game/Sounds/Fort_Monster_Deimos/Spawner/Fort_Deimos_SpawnerAmbientLarge_Att
 bOverrideConcurrency=true
 maxDistance=4000
 soundClassObject=/Game/Sounds/FortSoundClassesAndModes/Fort_Mission_IngameSounds
 volumeMultiplier=0.7
 concurrencyOverrides {maxCount=4, resolutionRule=PreventNew}
}
