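# Sound cue for the explosion of the Deimos monster's ice spawner.
# A single wave player feeds a Modulator node at default settings, which varies pitch and volume on each play.
# It uses the large explosives sound class at reduced volume and borrows the large spawner ambient attenuation from the base Spawner folder.
# A concurrency rule limits it to one instance per owner, stopping the oldest.
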SoundCue /Game/Sounds/Fort_Monster_Deimos/Spawner_Ice/DeimosSpawner_Ice_Explode_Cue
root=Modulator_0
nodes [
 {id=Modulator_0, children [WavePlayer_2]}
 {id=WavePlayer_2, properties {soundWaveAssetPtr=&/DeimosSpawner_Ice_Explode}}
]
properties {
 attenuationSettings=/Game/Sounds/Fort_Monster_Deimos/Spawner/Fort_Deimos_SpawnerAmbientLarge_Att
 maxDistance=4000
 soundClassObject=/Game/Sounds/FortSoundClassesAndModes/Fort_Explosives_Large_PSM
 volumeMultiplier=0.284
 concurrencySet [/Game/Sounds/Concurrency_Rules/OneMaximum_Owner_StopOldest]
}
