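# Sound cue for the Stage 2 explosion of the ice variant of the Deimos monster spawner.
# A single wave player feeds the output directly, reusing the base spawner's Stage 2 explosion wave rather than an ice-specific one.
# The cue boosts volume, routes to the enemy spawn sound class, and uses the base spawner's large ambient attenuation.
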SoundCue /Game/Sounds/Fort_Monster_Deimos/Spawner_Ice/DeimosSpawner_Stage2Explosion_Ice_Cue
root=WavePlayer_0
nodes [{id=WavePlayer_0, properties {soundWaveAssetPtr=/Game/Sounds/Fort_Monster_Deimos/Spawner/DeimosSpawner_Stage2Explosion}}]
properties {
 attenuationSettings=/Game/Sounds/Fort_Monster_Deimos/Spawner/Fort_Deimos_SpawnerAmbientLarge_Att
 maxDistance=4000
 soundClassObject=/Game/Sounds/FortSoundClassesAndModes/Fort_Enemy_Spawn
 volumeMultiplier=2
}
