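# Sound cue for the stage 3 buildup of the Deimos monster's ice spawner variant.
# A single wave player plays the buildup wave from the standard Spawner folder, with no ice-specific wave or further processing.
# It uses the spawner's large ambient attenuation, slightly reduced volume and a third-person ranged weapon fire sound class.
# A concurrency rule limits it to one instance per owner, stopping the oldest.
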SoundCue /Game/Sounds/Fort_Monster_Deimos/Spawner_Ice/DeimosSpawner_Stage3Buildup_Ice_Cue
root=WavePlayer_1
nodes [{id=WavePlayer_1, properties {soundWaveAssetPtr=/Game/Sounds/Fort_Monster_Deimos/Spawner/DeimosSpawner_Stage3Buildup}}]
properties {
 attenuationSettings=/Game/Sounds/Fort_Monster_Deimos/Spawner/Fort_Deimos_SpawnerAmbientLarge_Att
 maxDistance=4000
 soundClassObject=&Fort_WeaponRanged_Fire_3P_PSM
 volumeMultiplier=0.8
 concurrencySet [/Game/Sounds/Concurrency_Rules/OneMaximum_Owner_StopOldest]
}
